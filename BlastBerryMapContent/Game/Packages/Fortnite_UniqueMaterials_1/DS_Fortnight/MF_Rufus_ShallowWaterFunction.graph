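# Shallow water material function that outputs material attributes, their individual channels and a player-distance falloff.
# World-space panned, distorted normal maps blend and flatten with camera depth, with an optional skydive-distance normal layer.
# Color combines parallax-offset debris, a vertex-color transition, sand and Fresnel-tinted deep water, masked by player distance over ambient color, with time-of-day-scaled emissive.
# Mobile and forward paths use simpler vertex-interpolated panning normals and color without emissive.
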MaterialFunction /BlastBerryMapContent/Game/Packages/Fortnite_UniqueMaterials_1/DS_Fortnight/MF_Rufus_ShallowWaterFunction
graph {nodes [
  Constant3Vector() {constant=<rgba 0 0 1 0>}
  Constant() {r=1}
  Constant() {r=1}
  Constant() {r=0.5}
  n1 = CollectionParameter() {collection=&FortniteMaterialParameters, parameterName=PlayerPosition}
  n2 = Distance(A=ComponentMask(None=@n1) {b=true, g=true, r=true}, B=WorldPosition().XYZ)
  n3 = Clamp(None=Divide(A=Subtract(A=@n2) {constB=16}) {constB=64})
  FunctionOutput(None=@n3) {outputName=Distance falloff}
  n4 = ScalarParameter() {defaultValue=0.4, desc="water is 0.4.  Loot lake is 0.12", parameterName=Specular}
  n5 = ScalarParameter() {desc=Loot lake is 0.05, parameterName=Roughness}
  n6 = CollectionParameter() {
   collection=&FortniteMaterialParameters
   parameterName=FortniteActiveTimeOfDay
  }
  n7 = VectorParameter() {group=Deep Water, parameterName=Time of day deep water emissive values, defaultValue=<rgba 0 0.6192 0 0>}
  n8 = DotProduct(A=@n6, B=AppendVector(A=@n7.RGB, B=@n7.RGB))
  n9 = VertexColor()
  n10 = OneMinus(None=Multiply(A=@n9.R, B=@n9.R))
  n11 = ComponentMask(None=WorldPosition().XYZ) {g=true, r=true}
  n12 = Divide(A=@n11, B=ScalarParameter() {defaultValue=603.1323, parameterName=UVSize})
  n13 = Add(A=Multiply(A=@n12) {constB=0.8556}) {constB=0.3215}
  n14 = Add(A=@n12) {constB=0.1255}
  n15 = ComponentMask(None=TextureSample(UVs=Panner(Coordinate=@n14) {bFractionalPart=true, speedX=0.0208, speedY=-0.0278}) {samplerType=SAMPLERTYPE_Normal, texture=/Game/Global/Textures/Default/Tileable_Patterns/T_Water2_N}.RGB) {g=true, r=true}
  n16 = TextureSample(UVs=Add(A=Panner(Coordinate=@n13) {bFractionalPart=true, speedX=0.0625, speedY=0.0417}, B=Multiply(A=@n15) {constB=0.035})) {samplerType=SAMPLERTYPE_Normal, texture=&T_Water2_N}
  n17 = Multiply(A=ComponentMask(None=@n16.RGB) {g=true, r=true}, B=ScalarParameter() {defaultValue=0.0442, group=Deep Water, parameterName=Noise 1}) {constB=0.04}
  n18 = Add(A=Panner(Coordinate=@n12) {bFractionalPart=true, speedX=-0.025, speedY=-0.0375}) {constB=0.333}
  n19 = Multiply(A=ComponentMask(None=@n16.RGB) {g=true, r=true}, B=ScalarParameter() {defaultValue=0.0442, group=Deep Water, parameterName=Noise Far}) {constB=0.04}
  n20 = LinearInterpolate(
   A=TextureSample(UVs=Add(A=@n17, B=@n18)) {samplerType=SAMPLERTYPE_Normal, texture=&T_Water2_N}.RGB
   B=TextureSample(UVs=Add(A=@n19, B=@n18)) {samplerType=SAMPLERTYPE_Normal, texture=&T_Water2_N}.RGB
   Alpha=MaterialFunctionCall(
    Fade Length=ScalarParameter() {defaultValue=15000, group=Distance Blending, parameterName=Far Fade Length}
    Fade Offset=ScalarParameter() {defaultValue=7500, group=Distance Blending, parameterName=Far Fade Start}
   ) {materialFunction=&CameraDepthFade}.Result
  )
  n21 = Abs(None=MaterialFunctionCall(f2=@n20) {materialFunction=&AddComponents}.f2)
  n22 = Normalize(VectorInput=MaterialFunctionCall(BaseNormal=@n16.RGB, AdditionalNormal=@n20) {materialFunction=&BlendAngleCorrectedNormals}.Result)
  n23 = Add(
   A=Divide(A=@n11, B=ScalarParameter() {defaultValue=530.4807, group=Deep Water, parameterName=Rock Tiling})
   B=Multiply(A=ComponentMask(None=@n22) {g=true, r=true}, B=ScalarParameter() {defaultValue=0.1, group=Deep Water, parameterName=Ground Distortion RRRR})
  )
  n24 = CameraVectorWS()
  n25 = FunctionInput() {bUsePreviewValueAsDefault=true, inputName=Height Offset Amount, inputType=FunctionInput_Scalar, previewValue=<1 1 0 1>}
  n26 = Multiply(A=ComponentMask(None=Normalize(VectorInput=@n24)) {g=true, r=true}, B=@n25)
  n27 = Add(A=@n23, B=Multiply(A=@n26, B=Constant() {r=-0.15}))
  n28 = ShadingPathSwitch(
   Default=Multiply(A=TextureSample(UVs=@n27) {desc=floating debris, texture=/Game/Global/Textures/Default/PointSplatter/WM_StarField_D}.RGB) {constB=5}
   Mobile=Constant()
  ) {desc=Migrated from FeatureLevelSwitch}
  n29 = Add(A=Multiply(A=Power(Base=@n10) {constExponent=5}, B=OneMinus(None=@n21)), B=@n28)
  n30 = FunctionInput() {bUsePreviewValueAsDefault=true, inputName=Ground texture mask, inputType=FunctionInput_Scalar, previewValue=<1 0 0 1>}
  n31 = Add(
   A=Multiply(A=StaticSwitchParameter(True=@n29, False=@n28) {defaultValue=true, parameterName=UseStreamLakeTransiton}, B=@n30)
   B=VectorParameter() {group=Deep Water, parameterName=BAse Sand color, defaultValue=<rgba 0.0119 0.0302 0.04 1>}.RGB
  )
  n32 = Multiply(B=@n31)
  n33 = VectorParameter() {group=Deep Water, parameterName=Deep Water Color 1, defaultValue=<rgba 0.271 0.6224 0.635 0>}
  n34 = VectorParameter() {group=Deep Water, parameterName=Deep Water Color 2, defaultValue=<rgba 0.5718 0.8684 1 0.3>}
  n35 = MaterialFunctionCall(
   Invert Fresnel=StaticBool() {value=true}
   Power=ScalarParameter() {defaultValue=3, group=Deep Water, parameterName=POWER 33}
   Clamp Fresnel Dot Product=StaticBool() {value=true}
  ) {materialFunction=&Fresnel_Function}
  n36 = Multiply(
   A=Multiply(A=LinearInterpolate(A=@n33.RGB, B=@n34.RGB, Alpha=@n35.Result), B=LinearInterpolate(A=@n33.A, B=@n34.A, Alpha=@n35.Result))
   B=@n31
  )
  n37 = VectorParameter() {group=Deep Water, parameterName=Ambient Deep Water Color, defaultValue=<rgba 0.0038 0.0081 0.015 1>}
  n38 = Add(
   A=Multiply(
    A=@n3
    B=StaticSwitchParameter(True=@n32, False=@n36) {desc="Desert doesn't read fresnel properly so removing from there.  ", group=Deep Water, parameterName=Desert}
   )
   B=@n37.RGB
  )
  n39 = Constant() {r=6000}
  n40 = Constant() {r=2000}
  n41 = MaterialFunctionCall(
   Normal=@n22
   Flatness=LinearInterpolate(
    A=ScalarParameter() {defaultValue=0.89, parameterName=NormalFlatten}
    B=ScalarParameter() {defaultValue=0.99, parameterName=NormalFlatten_2}
    Alpha=MaterialFunctionCall(Fade Length=@n39, Fade Offset=@n40) {materialFunction=&CameraDepthFade}.Result
   )
  ) {materialFunction=&FlattenNormal}
  n42 = LinearInterpolate(A=@n41.Result, B=NamedRerouteUsage(), Alpha=NamedRerouteUsage()) {constAlpha=1}
  n43 = ScalarParameter() {defaultValue=2, parameterName=Brightness_M}
  n44 = Constant3Vector()
  n45 = VectorParameter() {parameterName=MobileNoSpecularColor, defaultValue=<rgba 0.3264 0.4144 0.4844 1>}
  n46 = ComponentMask(None=WorldPosition().XYZ) {g=true, r=true}
  n47 = ScalarParameter() {defaultValue=3934.5361, group=M, parameterName=UVSize_M_1}
  n48 = Time()
  n49 = ScalarParameter() {defaultValue=0.05, group=M, parameterName=PanSpeed_M3}
  n50 = VertexInterpolator(VS=Panner(Coordinate=Divide(A=@n46, B=@n47), Time=Multiply(A=@n48, B=@n49)) {bFractionalPart=true, speedX=0.11, speedY=0.27})
  n51 = Divide(A=@n46, B=ScalarParameter() {defaultValue=1308.5537, group=M, parameterName=UVSize_M})
  n52 = Multiply(A=@n48, B=ScalarParameter() {defaultValue=0.1, group=M, parameterName=PanSpeed_M})
  n53 = TextureSample(UVs=VertexInterpolator(VS=Panner(Coordinate=@n51, Time=@n52) {bFractionalPart=true, speedX=0.1, speedY=0.33}).PS) {samplerType=SAMPLERTYPE_Normal, texture=&T_Water2_N}
  n54 = Multiply(A=@n48, B=ScalarParameter() {defaultValue=0.05, group=M, parameterName=PanSpeed_M2})
  n55 = TextureSample(UVs=VertexInterpolator(VS=Panner(Coordinate=@n51, Time=@n54) {bFractionalPart=true, speedX=0.09, speedY=0.22}).PS) {samplerType=SAMPLERTYPE_Normal, texture=&T_Water2_N}
  n56 = ScalarParameter() {defaultValue=0.937, parameterName=NormalFlaten_M}
  n57 = ScalarParameter() {defaultValue=1, parameterName=NormalFlaten_M2}
  n58 = MaterialFunctionCall(
   Fade Length=ScalarParameter() {defaultValue=5805.6235, parameterName=FadeLength_M}
   Fade Offset=ScalarParameter() {defaultValue=1429.6207, parameterName=FadeOffset_M}
  ) {materialFunction=&CameraDepthFade}
  n59 = MaterialFunctionCall(
   BaseNormal=MaterialFunctionCall(
    Normal=TextureSample(UVs=@n50.PS) {samplerType=SAMPLERTYPE_Normal, texture=/Game/Global/Textures/Default/Tileable_Patterns/T_WaterSoft_N}.RGB
    Flatness=ScalarParameter() {defaultValue=0.9115, parameterName=NormalFlaten_M3}
   ) {materialFunction=&FlattenNormal}.Result
   AdditionalNormal=MaterialFunctionCall(
    Normal=MaterialFunctionCall(BaseNormal=@n53.RGB, AdditionalNormal=@n55.RGB) {materialFunction=&BlendAngleCorrectedNormals}.Result
    Flatness=LinearInterpolate(A=@n56, B=@n57, Alpha=@n58.Result)
   ) {materialFunction=&FlattenNormal}.Result
  ) {materialFunction=&BlendAngleCorrectedNormals}
  n60 = Saturate(None=ComponentMask(None=@n59.Result) {r=true})
  n61 = MakeMaterialAttributes!(
   BaseColor=Add(A=Multiply(A=@n37.RGB, B=@n43), B=LinearInterpolate(A=@n44, B=@n45.RGB, Alpha=@n60))
   Metallic=Constant()
   Specular=@n4
   Roughness=@n5
   Anisotropy
   EmissiveColor
   Opacity
   OpacityMask
   Normal=@n59.Result
   Tangent
   WorldPositionOffset
   SubsurfaceColor
   ClearCoat
   ClearCoatRoughness
   AmbientOcclusion
   Refraction
   CustomizedUVs_0
   CustomizedUVs_1
   CustomizedUVs_2
   CustomizedUVs_3
   CustomizedUVs_4
   CustomizedUVs_5
   CustomizedUVs_6
   CustomizedUVs_7
   PixelDepthOffset
   ShadingModel
   Displacement
  )
  n62 = ShadingPathSwitch(
   Default=MakeMaterialAttributes(
    BaseColor=@n38
    Specular=@n4
    Roughness=@n5
    EmissiveColor=Multiply(A=@n38, B=@n8)
    Normal=StaticSwitchParameter(True=@n42, False=@n41.Result) {desc=Enable if Loot Lake material, parameterName=LakeSkydive?}
   )
   Forward=@n61
   Mobile=@n61
  )
  FunctionOutput(None=@n62) {description=Base color specular roughness emissive and normal, outputName=MA Result}
  n63 = BreakMaterialAttributes!(Attr=@n62)
  FunctionOutput(None=@n63.BaseColor) {description=Base color specular roughness emissive and normal, outputName=Base Color, sortPriority=1}
  FunctionOutput(None=@n63.Specular) {description=Base color specular roughness emissive and normal, outputName=Spec, sortPriority=2}
  FunctionOutput(None=@n63.Roughness) {description=Base color specular roughness emissive and normal, outputName=Roughness, sortPriority=3}
  FunctionOutput(None=@n63.EmissiveColor) {description=Base color specular roughness emissive and normal, outputName=Emissive, sortPriority=4}
  FunctionOutput(None=@n63.Normal) {description=Base color specular roughness emissive and normal, outputName=Normal, sortPriority=5}
  n64 = Multiply(
   A=ComponentMask(None=@n16.RGB) {g=true, r=true}
   B=ScalarParameter() {defaultValue=0.0442, group=Distance Blending, parameterName=Noise Skydive}
  ) {constB=0.04}
  n65 = ScalarParameter() {defaultValue=9964.834, group=Distance Blending, parameterName=UVSize_skydive}
  n66 = Add(A=Panner(Coordinate=Divide(A=@n11, B=@n65)) {bFractionalPart=true, speedX=-0.005, speedY=-0.008}) {constB=0.333}
  NamedRerouteDeclaration(None=TextureSample(UVs=Add(A=@n64, B=@n66)) {samplerType=SAMPLERTYPE_Normal, texture=&T_Water2_N}.RGB) {name=Skydive_Normals, nodeColor=<rgba 0 0.2734 1 1>}
  NamedRerouteDeclaration(None=MaterialFunctionCall(
    Fade Length=ScalarParameter() {defaultValue=43732.9492, group=Distance Blending, parameterName=Skydive_Distance}
    Fade Offset=ScalarParameter() {defaultValue=500, group=Distance Blending, parameterName=Skydive_Offset}
   ) {materialFunction=&CameraDepthFade}.Result) {name=SkyDive_DistanceMask, nodeColor=<rgba 1 0.9141 0 1>}
  OneMinus(None=Clamp(None=DotProduct(A=@n20, B=@n16.RGB)))
  n67 = FunctionInput() {bUsePreviewValueAsDefault=true, inputName=Texture mip bias amount, inputType=FunctionInput_Scalar, previewValue=<0 1 0 1>}
  n68 = Add(A=@n23, B=Multiply(A=Constant() {r=-0.3968}, B=@n26))
  n69 = TextureSample!(UVs=@n68, Tex, Bias=Add(A=Constant() {r=2}, B=@n67), Apply View MipBias) {
   mipValueMode=TMVM_MipBias
   samplerSource=SSM_Wrap_WorldGroupSettings
   texture=/Game/Packages/DS_Fortnite_TRV_Terrain/Textures/T_Transylvania_HF_Mud_1_D
  }
  n70 = WorldPosition()
  n71 = TextureSample!(UVs=Divide(A=ComponentMask(None=@n70.XYZ) {g=true, r=true}) {constB=1024}, Tex, Apply View MipBias) {
   samplerSource=SSM_Wrap_WorldGroupSettings
   samplerType=SAMPLERTYPE_LinearGrayscale
   texture=/Game/Global/Textures/Default/Noise/T_SwirlNoise_N_height
  }
  Reroute(None=LinearInterpolate(
    A=TextureSample!(UVs=@n68, Tex, Bias=@n67, Apply View MipBias) {mipValueMode=TMVM_MipBias, samplerSource=SSM_Wrap_WorldGroupSettings, texture=/Game/Packages/DS_Fortnight/pebbles_blurred}.RGB
    B=@n69.RGB
    Alpha=@n71.R
   ))
  n72 = MaterialFunctionCall(
   Fade Length=ScalarParameter() {defaultValue=15000, group=Distance Blending, parameterName=Far Fade Length}
   Fade Offset=ScalarParameter() {defaultValue=7500, group=Distance Blending, parameterName=Far Fade Start}
  ) {materialFunction=&CameraDepthFade}
  Saturate(None=Multiply(A=@n35.Result, B=OneMinus(None=@n72.Result)))
 ]}
