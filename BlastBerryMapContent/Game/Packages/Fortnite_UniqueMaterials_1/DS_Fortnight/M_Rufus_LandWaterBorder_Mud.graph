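# Masked material for a muddy shoreline border where land meets water, usable on spline, instanced and Nanite meshes.
# World-aligned mud with detail normals blends into a shallow water function layer through gradient-driven height lerps, with panning wave normals and mud-alpha masking.
# Dithered, height-broken opacity fades the edges, and world position offset lifts the strip with an optional waterline hide.
# Static switches swap in a frozen lake variant or lower border verts; feature-level switches simplify blending on mobile.
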Material /BlastBerryMapContent/Game/Packages/Fortnite_UniqueMaterials_1/DS_Fortnight/M_Rufus_LandWaterBorder_Mud
graph {
 properties {
  bCastRayTracedShadows=false
  bTangentSpaceNormal=false
  bUseMaterialAttributes=true
  bUsedWithInstancedStaticMeshes=true
  bUsedWithNanite=true
  bUsedWithSplineMeshes=true
  blendMode=BLEND_Masked
  physMaterial=/Game/FortressPhysicalMaterials/Water
  pixelDepthOffsetMode=PDOM_Legacy
  translucencyPass=MTP_BeforeDOF
  usedShadingModels=Default Lit
 }
 nodes [
  Constant() {r=1}
  Constant() {r=0.05}
  Constant() {r=0.3}
  n1 = ScalarParameter() {defaultValue=-0.3596, group=Rufus, parameterName=Waterline_Start_Opacity}
  n2 = ScalarParameter() {defaultValue=1.3865, group=Rufus, parameterName=WaterlineEnd_Opacity}
  n3 = MaterialFunctionCall() {materialFunction=&LinearGradient}
  n4 = Saturate(None=OneMinus(None=LinearInterpolate(A=@n1, B=@n2, Alpha=@n3.VGradient)))
  MaterialFunctionCall(Alpha Threshold=@n4) {materialFunction=&DitherTemporalAA}
  n5 = TextureObjectParameter() {group=Mud, parameterName=Mud_Diffuse, samplerSource=SSM_Wrap_WorldGroupSettings, texture=/Game/Packages/DS_Fortnight/WhiteRockCustom_D}
  n6 = ScalarParameter() {defaultValue=900, group=Mud, parameterName=MudTextureTilingFrequency}
  n7 = MaterialFunctionCall(TextureObject=@n5, TextureSize=@n6, Export Float 4=StaticBool() {value=true}) {materialFunction=&WorldAlignedTexture}
  n8 = Multiply(
   A=VectorParameter() {group=Mud, parameterName=MudTint, defaultValue=<rgba 0.155 0.1348 0.1124 0>}.RGB
   B=ComponentMask(None=@n7.Z Texture) {b=true, g=true, r=true}
  )
  n9 = ConstantBiasScale(None=Sine(None=Add() {constA=0.2, constB=0.3})) {scale=0.25}
  n10 = Constant3Vector() {constant=<rgba 0 0 1 0>}
  n11 = TextureCoordinate()
  n12 = AppendVector(A=Constant() {r=1}, B=ScalarParameter() {defaultValue=0.5, parameterName=ShoreWaveTile})
  n13 = Constant2Vector() {g=1}
  n14 = ScalarParameter() {defaultValue=0.05, parameterName=ShoreWaveSpeed}
  n15 = TextureSampleParameter2D!(UVs=Panner(Coordinate=Multiply(A=@n11, B=@n12), Speed=Multiply(A=@n13, B=@n14)) {speedY=0.05}, Bias, Apply View MipBias) {
   mipValueMode=TMVM_MipBias
   parameterName=Normal_1
   samplerType=SAMPLERTYPE_Normal
   texture=/Game/Global/Textures/Default/Tileable_Patterns/T_Water2_N
  }
  n16 = Clamp(None=LinearInterpolate(
    A=ScalarParameter() {defaultValue=-0.3596, parameterName=Transition_M1}
    B=ScalarParameter() {defaultValue=1.3865, parameterName=Transition_M2}
    Alpha=MaterialFunctionCall() {materialFunction=&LinearGradient}.VGradient
   ))
  n17 = Saturate(None=Add(A=@n16, B=ScalarParameter() {defaultValue=1, parameterName=FlattenShoreWaves}))
  n18 = MaterialFunctionCall!(Height Offset Amount, Texture mip bias amount, Ground texture mask) {materialFunction=&/MF_Rufus_ShallowWaterFunction}
  n19 = LinearInterpolate(A=Transform(None=LinearInterpolate(A=@n10, B=@n15.RGB, Alpha=@n17)), B=@n18.Normal, Alpha=@n16)
  n20 = Multiply(
   A=Add(A=MaterialFunctionCall(Float3=@n19) {materialFunction=&BreakOutFloat3Components}.B) {constB=-1}
   B=ScalarParameter() {defaultValue=3, parameterName=WaveWaterHeight}
  )
  n21 = WorldPosition()
  n22 = TextureSample!(
   UVs=Divide(A=ComponentMask(None=@n21.XYZ) {g=true, r=true}, B=ScalarParameter() {defaultValue=512, group=Mud, parameterName=SandTextureTilingFrequency})
   Tex
   Bias=ScalarParameter() {defaultValue=2, parameterName=SandHeightMipBias}
   Apply View MipBias
  ) {mipValueMode=TMVM_MipBias, samplerType=SAMPLERTYPE_LinearColor, texture=/Game/Packages/Fortnite_UniqueMaterials_1/Textures/water2_DISP}
  n23 = ComponentMask(None=@n7.Z Texture) {a=true}
  n24 = LinearInterpolate(
   A=ScalarParameter() {parameterName=SandHeightMin}
   B=ScalarParameter() {defaultValue=1, parameterName=SandHeightMax}
   Alpha=StaticSwitchParameter(True=@n23, False=@n22.R) {parameterName=UseMudTextureAlphaAsSandMask}
  )
  n25 = MaterialFunctionCall(
   Transition Phase=Add(
    A=MaterialFunctionCall() {materialFunction=&LinearGradient}.VGradient
    B=Subtract(A=@n9) {constB=0.25}
   )
   Height Texture=LinearInterpolate(
    A=ScalarParameter() {defaultValue=1, parameterName=LerpHeight}
    B=Add(A=@n20, B=@n24)
    Alpha=ScalarParameter() {defaultValue=0.25, parameterName=HeightAlpha}
   )
   Contrast=ScalarParameter() {defaultValue=16, parameterName=Contrast}
  ) {materialFunction=&HeightLerp}
  n26 = Clamp(None=LinearInterpolate(
    A=ScalarParameter() {defaultValue=-1.1191, parameterName=Transition_M3}
    B=ScalarParameter() {defaultValue=6.6222, parameterName=Transition_M4}
    Alpha=MaterialFunctionCall() {materialFunction=&LinearGradient}.VGradient
   ))
  n27 = FeatureLevelSwitch(Default=@n25.Alpha, ES2=@n26, ES3_1=@n26, SM4_REMOVED=@n25.Alpha)
  n28 = LinearInterpolate(A=@n8, B=Multiply(A=ScalarParameter() {defaultValue=1, parameterName=MudDarkUnderWater}, B=@n8), Alpha=@n27)
  n29 = ScalarParameter() {defaultValue=0.02, parameterName=Sand Spec}
  n30 = Divide(
   A=ComponentMask(None=WorldPosition().XYZ) {g=true, r=true}
   B=ScalarParameter() {defaultValue=512, group=Mud, parameterName=SandNormalTilingFrequency}
  )
  n31 = TextureObjectParameter() {
   group=Mud
   parameterName=Mud_Normal
   samplerSource=SSM_Wrap_WorldGroupSettings
   samplerType=SAMPLERTYPE_Normal
   texture=/Game/Packages/DS_Fortnight/WhiteRockCustom_N
  }
  n32 = MaterialFunctionCall(TextureObject=@n31, TextureSize=@n6) {materialFunction=&WorldAlignedNormal}
  n33 = MaterialFunctionCall(
   BaseNormal=@n32.Z Texture
   AdditionalNormal=LinearInterpolate(
    A=TextureSample!(UVs=@n30, Tex, Apply View MipBias) {samplerType=SAMPLERTYPE_Normal, texture=/Game/Packages/DS_Fortnight/T_MudBeachNormal_01}.RGB
    B=Constant3Vector() {constant=<rgba 0 0 1 0>}
    Alpha=@n23
   )
  ) {materialFunction=&BlendAngleCorrectedNormals}
  n34 = QualitySwitch(Default=@n33.Result, Low=@n32.Z Texture, High=@n33.Result, Medium=@n33.Result)
  n35 = LinearInterpolate(A=ScalarParameter() {parameterName=Sand Roughness}, B=ScalarParameter() {defaultValue=2, parameterName=Sand Roughness Max}, Alpha=@n23)
  n36 = ScalarParameter() {defaultValue=-140, parameterName=DistortAmount}
  n37 = Add(A=WorldPosition().XYZ, B=Multiply(A=Multiply(A=@n19, B=@n36), B=@n25.Alpha))
  n38 = StaticBool() {value=true}
  n39 = OneMinus(None=ComponentMask(None=MaterialFunctionCall(TextureObject=@n5, TextureSize=@n6, WorldPosition=@n37, Export Float 4=@n38) {materialFunction=&WorldAlignedTexture}.Z Texture) {a=true})
  n40 = StaticSwitchParameter(
   True=LinearInterpolate(
    A=@n16
    B=MaterialFunctionCall(Transition Phase=@n16, Height Texture=@n39) {materialFunction=&HeightLerp}.Lerp Alpha No Contrast
    Alpha=ScalarParameter() {defaultValue=0.3, parameterName=MudAlphaToWaterBlend}
   )
   False=@n16
  ) {parameterName=UseMudTextureAlphaAsSandMask}
  n41 = MaterialFunctionCall!(
   Material=MaterialFunctionCall!(
    Base Material=MakeMaterialAttributes!(
     BaseColor=@n28
     Metallic
     Specular=@n29
     Roughness=@n35
     Anisotropy
     EmissiveColor
     Opacity
     OpacityMask
     Normal=@n34
     Tangent
     WorldPositionOffset
     SubsurfaceColor
     ClearCoat
     ClearCoatRoughness
     AmbientOcclusion
     Refraction
     CustomizedUVs_0
     CustomizedUVs_1
     CustomizedUVs_2
     CustomizedUVs_3
     CustomizedUVs_4
     CustomizedUVs_5
     CustomizedUVs_6
     CustomizedUVs_7
     PixelDepthOffset
     ShadingModel
     Displacement
    )
    Top Material=@n18.MA Result
    Alpha=FeatureLevelSwitch(Default=@n40, ES2=@n16, ES3_1=@n16, SM4_REMOVED=@n40)
   ) {materialFunction=/Engine/Functions/MaterialLayerFunctions/MatLayerBlend_Standard}.Blended Material
   Normal=LinearInterpolate(A=@n34, B=FeatureLevelSwitch(Default=@n19, ES2=@n18.Normal, ES3_1=@n18.Normal, SM4_REMOVED=@n19), Alpha=@n27)
  ) {materialFunction=/Engine/Functions/MaterialLayerFunctions/MatLayerBlend_ReplaceNormals}
  n42 = LinearInterpolate(
   A=Add(
    A=ScalarParameter() {defaultValue=-0.3596, group=Rufus, parameterName=Waterline_Start}
    B=ScalarParameter() {group=Rufus, parameterName=waterline_roughnessBlend}
   )
   B=ScalarParameter() {defaultValue=1.3865, group=Rufus, parameterName=WaterlineEnd}
   Alpha=MaterialFunctionCall() {materialFunction=&LinearGradient}.VGradient
  )
  MaterialFunctionCall!(Material=@n41.Result, NewRoughness=Constant(), Mask=OneMinus(None=Clamp(None=@n42))) {materialFunction=/Engine/Functions/MaterialLayerFunctions/MatLayerBlend_RoughnessOverride}
  MaterialFunctionCall!(Height Offset Amount, Texture mip bias amount, Ground texture mask) {materialFunction=/Game/WaterAndWind/Materials/Water/ShallowWaterFunction}
  Reroute(None=@n18.Normal)
  Reroute(None=@n16)
  n43 = MaterialFunctionCall() {materialFunction=/Game/Athena/WinterArt/Materials/MF_FrozenLake}
  n44 = Multiply(A=Constant3Vector() {constant=<rgba 0 0 1 0>}, B=ScalarParameter() {defaultValue=1, parameterName=Z offset amount})
  n45 = Constant3Vector() {constant=<rgba 0 0 2 0>}
  n46 = ComponentMask(None=WorldPosition().XYZ) {b=true}
  n47 = Constant() {r=10}
  n48 = CameraPositionWS()
  n49 = Saturate(None=Subtract(A=Add(A=@n46, B=@n47), B=ComponentMask(None=@n48) {b=true}))
  n50 = LinearInterpolate(
   A=ScalarParameter() {defaultValue=-0.3596, group=Rufus, parameterName=Waterline_Start}
   B=ScalarParameter() {defaultValue=1.3865, group=Rufus, parameterName=WaterlineEnd}
   Alpha=MaterialFunctionCall() {materialFunction=&LinearGradient}.VGradient
  )
  n51 = StaticSwitchParameter(True=Multiply(A=LinearInterpolate(A=@n44, B=@n45, Alpha=@n49), B=Clamp(None=@n50)), False=@n44) {group=Rufus, parameterName=Rufus_HideWaterline}
  n52 = MaterialFunctionCall(Float2=TextureCoordinate()) {materialFunction=&BreakOutFloat2Components}
  n53 = Multiply(
   A=Distance(A=@n52.G, B=ScalarParameter() {defaultValue=0.5, parameterName=OpacityFalloffCenter})
   B=ScalarParameter() {defaultValue=2, parameterName=OpacityFalloff}
  ) {constB=2}
  n54 = Multiply(A=@n43.Mask, B=Constant() {r=3})
  n55 = Multiply(A=OneMinus(None=@n53)) {constB=2}
  n56 = MaterialFunctionCall(Alpha Threshold=MaterialFunctionCall(
    Transition Phase=Clamp(None=@n55)
    Height Texture=OneMinus(None=@n23)
    Contrast=ScalarParameter() {defaultValue=1, parameterName=Opacity mask Contrast}
   ) {materialFunction=&HeightLerp}.Alpha) {materialFunction=&DitherTemporalAA}
  n57 = StaticSwitchParameter(
   True=MaterialFunctionCall(Transition Phase=@n53, Height Texture=LinearInterpolate(A=@n23, B=@n54, Alpha=@n52.G), Contrast=Constant() {r=-2}) {materialFunction=&HeightLerp}.Alpha
   False=@n56.Result
  ) {group=Winter, parameterName=FrozenLakeWater}
  n58 = MaterialFunctionCall!(
   Material=StaticSwitchParameter(
    True=@n43.Result
    False=MaterialFunctionCall!(Material=@n41.Result, WorldPositionOffset=@n51) {materialFunction=/Engine/Functions/MaterialLayerFunctions/MatLayerBlend_AddWorldPositionOffset}.Blended Material
   ) {group=Winter, parameterName=FrozenLakeWater}
   NewOpacityMask=@n57
   Mask
  ) {materialFunction=/Engine/Functions/MaterialLayerFunctions/MatLayerBlend_OverrideOpacityMask}
  n59 = MaterialFunctionCall!(Render To Texture Pass, WS Flow Map Dir=Constant2Vector(), Opacity, Default Mat Att=@n58.Blended Material) {materialFunction=/Game/WaterAndWind/Materials/Water/WaterRenderToTextureMaterialAtt}
  n60 = FeatureLevelSwitch(Default=@n59.Result, ES2=@n58.Blended Material, ES3_1=@n58.Blended Material, SM4_REMOVED=@n59.Result)
  n61 = GetMaterialAttributes(None=@n60) {attributeGetTypes [#1]}
  n62 = ComponentMask(None=@n61.World Position Offset) {g=true, r=true}
  n63 = Subtract(A=ComponentMask(None=@n61.World Position Offset) {b=true}, B=ScalarParameter() {defaultValue=200, parameterName=BorderVerts_ZOffset})
  StaticSwitchParameter(
   True=SetMaterialAttributes!(MaterialAttributes=@n60, World Position Offset=AppendVector(A=@n62, B=@n63)) {attributeSetTypes [#1]}
   False=@n60
  ) {parameterName=Athena_HideBorderVerts}
  StaticSwitchParameter(True=Multiply(A=@n56.Result, B=@n4), False=@n57) {group=Rufus, parameterName=Rufus_HideWaterline}
  TextureCoordinate() {uTiling=0.5, vTiling=0.25}
 ]
}
